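; Parent Blueprint for fellable props such as trees that, on death, spawn a physics log and optional stump.
; BeginPlay caches spawn transforms and, on the server when felling is disabled, swaps in a replacement prop after a random delay.
; Server death picks a fall direction (preset, slope/random, toward enemy or away from damage), then impulses the log.
; It also carries over Chrome, fires telemetry and replicates the fall vector for client VFX.
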(event Custom|Reset%20Starting%20World%20Transform
  ;[1 1 1] Starting world transform
  (Variables|FellingResults|SetStartingWorldTransform (Transformation|GetWorldTransform (Variables|Debug|GetStaticMeshComponent)))
  (bind _physicsobjectspawnlocation (Variables|Default|GetPhysicsObjectSpawnLocation)
    (:then
      (if (Math|Vector|VectorIsNearlyZero (.location (Transformation|GetWorldTransform _physicsobjectspawnlocation)))
        ;[1 0 0] Fallback for invalid Spawn Transform or Scene Component.
        (Fortnite|FortShippingLog self "Error Caching Physics Log Spawn Transform. SpawnLocation Transform Values are 0" true)
        (Development|PrintString "Error Caching Physics Log Spawn Transform. SpawnLocation Transform Values are 0")
        (else
          (Variables|FellingResults|SetPhysicsObjectSpawnTransform (Transformation|GetWorldTransform _physicsobjectspawnlocation)))))
    (:else
      (Fortnite|FortShippingLog self "Error Caching Physics Log Spawn Transform. PhysicsObjectSpawnLocation Scene Component Invalid" true)
      (Development|PrintString "Error Caching Physics Log Spawn Transform. PhysicsObjectSpawnLocation Scene Component Invalid")))
  (Variables|FellingResults|SetPhysicsObjectSpawnTransform (Math|Transform|MakeTransform (+ (.location (Transformation|GetWorldTransform (Variables|Debug|GetStaticMeshComponent))) "0,0,900.000000") (.rotation (Transformation|GetWorldTransform (Variables|Debug|GetStaticMeshComponent))) (.scale (Transformation|GetWorldTransform (Variables|Debug|GetStaticMeshComponent))))))

(event EventBeginPlay
  (bind (_location _rotation _scale) (.location (Transformation|GetActorTransform)))
  (bind (_scale_x _scale_y _scale_z) (Math|Vector|BreakVector _scale))
  (bind (_randomscalerange_x _randomscalerange_y) (Math|Vector2D|BreakVector2D (Variables|GetRandomScaleRange_X)))
  (bind (_replacementscalerangemap_x _replacementscalerangemap_y) (Math|Vector2D|BreakVector2D (Variables|GetReplacementScaleRangeMap_X)))
  (bind _replacementdelayrange_x (Math|Vector2D|BreakVector2D (Variables|GetReplacementDelayRange_X)))
  (bind (_replacementdelayrange_x_2 _replacementdelayrange_y) (Math|Vector2D|BreakVector2D (Variables|GetReplacementDelayRange_X)))
  (|CallParentFunction)
  (Utilities|FlowControl|Sequence
    (:then_0
      (if (Networking|IsServer)
        (if (not (CallFunction|GetFellableEnabled))
          ;[1 1 1] Fellable is disabled, replace with normal prop
          (Variables|Replacement|SetReplacementDelay (Math|Random|RandomFloatinRange _replacementdelayrange_x_2 _replacementdelayrange_y))
          (Utilities|FlowControl|Sequence
            (:then_0
              (Utilities|FlowControl|Delay (Variables|Replacement|GetReplacementDelay))
              (Building|SpawnBuildingActor (Variables|FellingConfig|GetDisabledActorReplacementClass) (Math|Transform|MakeTransform _location _rotation (Math|Conversions|ToVector(Float) (Math|Float|MapRangeUnclamped _scale_z _randomscalerange_x _randomscalerange_y _replacementscalerangemap_x _replacementscalerangemap_y)))))
            (:then_1
              (Actor|SetLifeSpan (+ (Variables|Replacement|GetReplacementDelay) _replacementdelayrange_x)) ; Destroy this actor
            )))))
    (:then_1
      ;[1 1 1] Starting world transform
      (Variables|FellingResults|SetStartingWorldTransform (Transformation|GetWorldTransform (Variables|Debug|GetStaticMeshComponent)))
      (bind _physicsobjectspawnlocation (Variables|Default|GetPhysicsObjectSpawnLocation)
        (:then
          (if (Math|Vector|VectorIsNearlyZero (.location (Transformation|GetWorldTransform _physicsobjectspawnlocation)))
            ;[1 0 0] Fallback for invalid Spawn Transform or Scene Component.
            (Fortnite|FortShippingLog self "Error Caching Physics Log Spawn Transform. SpawnLocation Transform Values are 0" true)
            (Development|PrintString "Error Caching Physics Log Spawn Transform. SpawnLocation Transform Values are 0")
            (else
              (Variables|FellingResults|SetPhysicsObjectSpawnTransform (Transformation|GetWorldTransform _physicsobjectspawnlocation)))))
        (:else
          (Fortnite|FortShippingLog self "Error Caching Physics Log Spawn Transform. PhysicsObjectSpawnLocation Scene Component Invalid" true)
          (Development|PrintString "Error Caching Physics Log Spawn Transform. PhysicsObjectSpawnLocation Scene Component Invalid")))
      (Variables|FellingResults|SetPhysicsObjectSpawnTransform (Math|Transform|MakeTransform (+ (.location (Transformation|GetWorldTransform (Variables|Debug|GetStaticMeshComponent))) "0,0,900.000000") (.rotation (Transformation|GetWorldTransform (Variables|Debug|GetStaticMeshComponent))) (.scale (Transformation|GetWorldTransform (Variables|Debug|GetStaticMeshComponent))))))))

(event Custom|Reenable%20Impulse%20via%20damage
  (FortPhysicsSimulationLibrary|SetEnableImpulseonDamageforPhysicsObject (Variables|FellingResults|GetSpawnedPhysicsObject) true))

(event Custom|Object%20Felled%20on%20Server)

(event Custom|FellingVectorReplicated)

(event EventOnDeathServer
  (params (Damage :type Float%20%28single-precision%29) (DamageTags :type Gameplay%20Tag%20Container%20Structure%20%28by%20ref%29) (Momentum :type Vector :default "0, 0, 0") (HitInfo :type Hit%20Result%20Structure%20%28by%20ref%29) (InstigatedBy :type Controller%20Object%20Reference) (DamageCauser :type Actor%20Object%20Reference) (EffectContext :type Gameplay%20Effect%20Context%20Handle%20Structure))
  (bind (_x _y) (Math|Vector2D|BreakVector2D (* (Variables|FellingResults|GetFallVector) (ScalableFloat|GetValueatLevel (Variables|FellingDataRows|GetFellingImpulseMagnitude)))))
  (bind _fellingimpulsemagnitude (Variables|FellingDataRows|GetFellingImpulseMagnitude))
  (bind _returnvalue (ScalableFloat|GetValueatLevel _fellingimpulsemagnitude))
  (bind _fallvector (Variables|FellingResults|GetFallVector))
  (bind _returnvalue_2 (* _fallvector _returnvalue))
  (bind (_x_2 _y_2) (Math|Vector2D|BreakVector2D _returnvalue_2))
  (bind _physicsobjectspawntransform (Variables|FellingResults|GetPhysicsObjectSpawnTransform))
  (bind _location (.location _physicsobjectspawntransform))
  (bind _spawnedphysicsobject (Variables|FellingResults|GetSpawnedPhysicsObject))
  (bind _felleddamagecauser (Variables|FellingResults|GetFelledDamageCauser))
  (Variables|FellingResults|SetFellableHitInfo HitInfo)
  (Variables|FellingResults|SetFelledDamageTags DamageTags)
  (Variables|FellingResults|SetFelledDamageCauser DamageCauser)
  (CallFunction|CheckforChrome)
  (if (CallFunction|GetFellableEnabled)
    ;[1 1 1] Spawn Log
    (if (ScalableFloat|GetValueasBool (Variables|FellingDataRows|GetPhysicsSpawnEnabled))
      ;;[1 1 1] Check if it's been moved (for example, by the Prop Mover)
      (if (not (== (Variables|FellingResults|GetStartingWorldTransform) (Transformation|GetWorldTransform (Variables|Debug|GetStaticMeshComponent))))
        (CallFunction|ResetStartingWorldTransform))
      (Utilities|Casting|CastToFortAthenaVehicle DamageCauser
        (:then
          _)
        (:CastFailed
          (CallFunction|TrySpawnStump)))
      (if (Math|Vector|VectorIsNearlyZero _location)
        ;;[1 0 0] Fallback for invalid SpawnLocation. Perhaps an asset that was placed before the refactor could have an invalid location
        (Fortnite|FortShippingLog self "Error Spawning Physics Log. Values not cached in BeginPlay" true)
        (Development|PrintString "Error Spawning Physics Log. Values not cached in BeginPlay")
        (else
          (bind _returnvalue_3 (Building|SpawnBuildingActor (Variables|FellingConfig|GetSpawnedPhysicsObjectClass) (Math|Transform|MakeTransform (+ _location (Variables|FellingConfig|GetSpawnPhysicsObjectOffsetLocation)) (.rotation _physicsobjectspawntransform) (Math|Vector|MakeVector (Math|Float|Lerp 1.0 (.x (.scale _physicsobjectspawntransform)) (.x (Variables|FellingConfig|GetSpawnedObjectMatchScaleAxes))) (Math|Float|Lerp 1.0 (.y (.scale _physicsobjectspawntransform)) (.y (Variables|FellingConfig|GetSpawnedObjectMatchScaleAxes))) (Math|Float|Lerp 1.0 (.z (.scale _physicsobjectspawntransform)) (.z (Variables|FellingConfig|GetSpawnedObjectMatchScaleAxes))))) 0 (Variables|FellingResults|GetFelledDamageCauser))) ; Spawn Log
          (bind _output_get (Variables|FellingResults|SetSpawnedPhysicsObject _returnvalue_3))
          (Utilities|IsValid _output_get
            (:Is%20Valid
              (PhysicsObject|WakeUpPhysicsObject _output_get "AlwaysInstant") ; Wake up physics object immediately here to allow impulse to be effectively applied. \nRemark: activation will be queued, and impulses in this tick will be lost.\nTodo: Improve physics object manager to allow for queuing impulses.
            )
            (:Is%20Not%20Valid
              _))
          (CallFunction|ApplyChrome _output_get)))
      (Curie|ChangeCurieContainerOwnership self _output_get)
      ;;[1 1 1] Set up Physics Object
      (Default|CallOnSpawnedPhysicsObject _output_get)
      (bind _returnvalue_4 (BuildingActor|PhysicsSimulationSupport|GetPhysicsObjectComponent _output_get))
      (bind _spawnedstumpobject (Variables|FellingResults|GetSpawnedStumpObject)
        (:then
          (PhysicsObject|AddtoIgnoredDamageActors _returnvalue_4 _spawnedstumpobject))
        (:else
          _))
      (CallFunction|GetTreeFeller _felleddamagecauser)
      (FortPhysicsSimulationLibrary|SetImpactInstigatorforPhysicsObject (Variables|FellingResults|GetSpawnedPhysicsObject) _felleddamagecauser)
      ;;[1 1 1] Briefly ignore so that users can't impulse with pickaxe immediately
      (FortPhysicsSimulationLibrary|SetEnableImpulseonDamageforPhysicsObject (Variables|FellingResults|GetSpawnedPhysicsObject))
      (Utilities|Time|SetTimerbyEvent (EventDispatchers|CreateEvent :!event Custom|Reenable%20Impulse%20via%20damage) (ScalableFloat|GetValueatLevel (Variables|FellingDataRows|GetFellingImpulseDisableTime)))
      (CallFunction|SetFallVectorXY) ; use class properties to set the \nfall behavior of the Fellable actor\n- Randomized\n- Fall away from damage\n- Target nearby players\n- etc.
      (Utilities|FlowControl|Sequence
        (:then_0
          (Variables|FellingResults|SetFallVector_FX (Math|Conversions|ToVector(Vector2D) _returnvalue_2))
          (bind _asstatic_mesh_component (Utilities|Casting|CastToStaticMeshComponent (Variables|/Script/Engine.Actor|GetRootComponent (Variables|FellingResults|GetSpawnedPhysicsObject))))
          (if (Transformation|DoesSocketExist _asstatic_mesh_component "FellingImpulseLocation")
            (FortPhysicsSimulationLibrary|AddPhysicsImpulseatLocation (Variables|FellingResults|GetSpawnedPhysicsObject) (Math|Vector|MakeVector _x_2 _y_2 (ScalableFloat|GetValueatLevel (Variables|FellingConfig|GetFelledImpulseZ))) (Transformation|GetSocketLocation _asstatic_mesh_component "FellingImpulseLocation") false "None" false 0 "(MaxSimSizeAllowed=Large,bRejectEmbeddedInLandscape=False)")
            (else
              (FortPhysicsSimulationLibrary|AddPhysicsImpulse _spawnedphysicsobject (Math|Vector|MakeVector _x _y (ScalableFloat|GetValueatLevel (Variables|FellingConfig|GetFelledImpulseZ))) false "None" false false 0 "(MaxSimSizeAllowed=Large,bRejectEmbeddedInLandscape=False)")
              (Development|PrintString (Utilities|String|BuildString(Name) "" "Physics Fellable: " (Utilities|String|StringToName (Utilities|GetObjectName _spawnedphysicsobject)) "missing Socket \"FellingImpulseLocation\", applying impulse to center"))))
          (FortPhysicsAnalytics|FireTelemetryEventonPhysicsFellableDestroyed (Fortnite|GetFortPlayerControllerfromActor (Variables|FellingResults|GetTree%20Feller)) self (Variables|FellingResults|GetFelledDamageCauser) (Variables|FellingResults|GetFelledDamageTags) (Variables|Default|GetFallTowardsEnemy)))
        (:then_1
          (CallFunction|ShouldDrawDebugFellingArrow))
        (:then_2
          (CallFunction|ObjectFelledOnServer))))))

(fn ApplyChrome
  (params (Target :type Actor%20Object%20Reference))
  (bind _returnvalue (Ability|GetAbilitySystemComponent Target))
  (bind _chromege (Variables|Default|GetChromeGE))
  (if (Utilities|IsValidClass _chromege)
    (Utilities|IsValid _returnvalue
      (:Is%20Valid
        (GameplayEffects|ApplyGameplayEffectToSelf :self _returnvalue :GameplayEffectClass _chromege :Level 0.0)))))

(fn BuildDataRegistryResolverScope
  ;[1 0.8475 0.2776] Gives the name of the volume as a scope to the data registry. Needed for calls to GetValueWithContext to work
  (return (Utilities|Array|MakeArray (SpatialGameplay|GetSpatialGameplayInterfaceFNameforActor (Minigame|GetMinigame self :!class /Script/FortniteGame.MinigameBlueprintLibrary))) 0 false))

(fn CheckForChrome
  (bind _self self)
  (bind _returnvalue (Ability|GetAbilitySystemComponent _self))
  (Utilities|IsValid _returnvalue
    (:Is%20Valid
      (bind _returnvalue_2 (GameplayEffects|GetActiveEffectswithAllTags _returnvalue "(GameplayTags=((TagName=\"Gameplay.Status.Chrome\")))"))
      (bind _array_element (Utilities|Array|ForEachLoopwithBreak _returnvalue_2)
        (:LoopBody
          (bind _output_get (Variables|Default|SetChromeGE (Utilities|GetClass (Ability|GameplayEffect|GetGameplayEffectfromActiveEffectHandle _array_element))))
          (Development|PrintString (Utilities|GetClassDisplayName _output_get)))))))

(fn GetFellableEnabled
  (return (ScalableFloat|GetValueasBoolwithContext (Variables|FellingDataRows|GetFellableEnabled))))

(fn GetCacheContextOverride
  (return 0))

(fn GetTreeFeller
  (params (Damage%20Causer :type Actor%20Object%20Reference))
  (bind _returnvalue (Actor|GetOwner Damage%20Causer))
  (Utilities|IsValid _returnvalue
    (:Is%20Valid
      (bind _output_get (Variables|FellingResults|SetTree%20Feller _returnvalue)))
    (:Is%20Not%20Valid
      (bind _returnvalue_2 (FortPhysicsSimulationLibrary|GetDamageInstigatorforPhysicsCollision Damage%20Causer))
      (Variables|FellingResults|SetTree%20Feller _returnvalue_2)))
  (return _output_get))

(fn LerpVector2D
  (params (A :type Vector%202D%20Structure) (B :type Vector%202D%20Structure) (Alpha :type Float%20%28double-precision%29))
  (bind (_returnvalue_x _returnvalue_y) (Math|Vector|BreakVector (Math|Vector|Lerp(Vector) (Math|Vector|MakeVector (Math|Vector2D|BreakVector2D A) (.Y (Math|Vector2D|BreakVector2D A))) (Math|Vector|MakeVector (Math|Vector2D|BreakVector2D B) (.Y (Math|Vector2D|BreakVector2D B))) Alpha)))
  (bind (_x _y) (Math|Vector2D|BreakVector2D A))
  (bind (_x_2 _y_2) (Math|Vector2D|BreakVector2D B))
  ;[0.2435 0.348 1 0.246] Simple Lerp for a 2d Vector. Using this rather than the function library one as EUW_FunctionLibrary pulls in fonts that we don't want.
  (return _returnvalue_x _returnvalue_y))

(fn OnRep_FallVector_FX
  (bind (_bblockinghit _binitialoverlap _time _distance _location) (Collision|BreakHitResult (Variables|FellingResults|GetFellableHitInfo)))
  (Networking|SwitchHasAuthority
    (:Remote
      (if (CallFunction|GetFellableEnabled)
        (if (> (Math|Vector|VectorLength (Math|Vector|MakeVector (.x (Variables|FellingResults|GetFallVector_FX)) (.y (Variables|FellingResults|GetFallVector_FX)))) 0)
          (bind _fellingvfxasset (Variables|FellingConfig|GetFellingVFXAsset)
            (:then
              (Effects|Components|ParticleSystem|SpawnFXAtLocation _fellingvfxasset _location)
              (CallFunction|FellingVectorReplicated))
            (:else)))))))

(fn SendFallenTreeStatEvent
  (params (DamageCauser :type Actor%20Object%20Reference))
  (bind _asfort_player_pawn (Utilities|Casting|CastToFortPlayerPawn (Actor|GetOwner DamageCauser))
    (:then
      (if (Utilities|IsValid (Fortnite|GetFortPlayerControllerfromActor _asfort_player_pawn) :!class /Script/Engine.KismetSystemLibrary)
        (Quest|SendComplexCustomStatEvent (Quests|GetQuestManager (Fortnite|GetFortPlayerControllerfromActor _asfort_player_pawn) "Athena") 0 "(GameplayTags=((TagName=\"Athena.Quests.LogsFallen\")))" "(GameplayTags=)" 1)
        (return true)
        (else
          (return false))))
    (:CastFailed
      (return false))))

(fn SetFallVectorXY
  (bind (_bblockinghit _binitialoverlap _time _distance _location _impactpoint _normal) (Collision|BreakHitResult (Variables|FellingResults|GetFellableHitInfo)))
  ;[0.2454 0.3266 0.8125] Set Fall Vector XY
  ; Determine the method of how we want the object to fall using the default parameters
  (bind (_outhit _returnvalue) (Collision|LineTraceByChannel (- (Transformation|GetActorLocation) (Variables|FellingConfig|GetStumpTraceStartOffset)) (- (Transformation|GetActorLocation) (Variables|FellingConfig|GetStumpTraceEndOffset)) "TraceTypeQuery8" false (select (Utilities|IsValid (Variables|FellingResults|GetSpawnedStumpObject) :!class /Script/Engine.KismetSystemLibrary) (Utilities|Array|MakeArray (Variables|FellingResults|GetSpawnedPhysicsObject) (Variables|FellingResults|GetSpawnedStumpObject)) (Utilities|Array|MakeArray (Variables|FellingResults|GetSpawnedPhysicsObject))) (select (Variables|FellingConfig|GetDrawFellingDebug) "Persistent" "None"))) ; Line trace ground for slope
  (Utilities|FlowControl|Sequence
    (:then_0
      ; Fall in a preset direction set by the instance
      (if (Variables|FellingConfig|GetbFallInPresetDirection)
        (Variables|FellingResults|SetFallVector (Variables|FellingConfig|GetPresetFallDirection))
        (return)))
    (:then_1
      ;;[1 1 1] Towards enemy
      (bind _output_get (Variables|Default|SetClosestFortPlayerPawnByTeamAffiliation (Fortnite|GetClosestFortPlayerPawnbyTeamAffiliation (.Location (Collision|BreakHitResult _outhit)) (Variables|FellingResults|GetTree%20Feller)))))
    (:then_2
      ;;[1 1 1] Random / slope
      (bind _output_get_2 (Variables|FellingResults|SetRandomFallDirection (Math|Conversions|ToVector2D(Vector) (Math|Random|RandomUnitVector)))) ; Random
      (if _returnvalue
        (Variables|FellingResults|SetFallVector (Math|Vector2D|Normalize2D (Default|LerpVector2D (Math|Conversions|ToVector2D(Vector) (* (.ImpactNormal (Collision|BreakHitResult _outhit)) -1.0)) _output_get_2 (Variables|FellingConfig|GetSlopeBlend)))) ; Mixed downhill and random
        (else
          (Variables|FellingResults|SetFallVector _output_get_2) ; Random only
        ))
      (if (and (and (< (Class|AbilityFunctions|GetDistanceBetweenActors _output_get self) (Variables|FellingConfig|GetFallTargetingDistance)) (< (Math|Random|RandomFloat) (ScalableFloat|GetValueatLevel (Variables|FellingDataRows|GetFellingAutoTargetChance)))) (not (GameplayTags|HasAnyTags (Variables|FellingResults|GetFelledDamageTags) (Variables|Default|GetDamageTagsToPreventAutoTarget))))
        (if (Utilities|IsValid _output_get :!class /Script/Engine.KismetSystemLibrary)
          (Variables|Default|SetFallTowardsEnemy true)
          (Variables|FellingResults|SetFallVector (Math|Vector2D|Normalize2D (Math|Conversions|ToVector2D(Vector) (- (* (Transformation|GetActorLocation _output_get) "1.000000,1.000000,0.000000") (* (Transformation|GetActorLocation) "1.000000,1.000000,0.000000"))))) ; Fall towards enemy
          (else
            _))
        (else
          _))
      (if (GameplayTags|HasAnyTags (Variables|FellingResults|GetFelledDamageTags) (Variables|FellingConfig|GetFallAwayDamageTagList))
        (Variables|FellingResults|SetFallVector (Math|Vector2D|Normalize2D (Math|Conversions|ToVector2D(Vector) (Math|Vector|SelectVector (- (* (Transformation|GetActorLocation) "1.000000,1.000000,0.000000") (* (Transformation|GetActorLocation (Variables|FellingResults|GetTree%20Feller)) "1.000000,1.000000,0.000000")) (* _normal "-1.000000,-1.000000,0.000000") (GameplayTags|HasAnyTags (Variables|FellingResults|GetFelledDamageTags) (Variables|Default|GetPawnDamageCauserTagList))))))))))

(fn ShouldDrawDebugFellingArrow
  (if (Variables|FellingConfig|GetDrawFellingDebug)
    (Rendering|Debug|DrawDebugArrow (Transformation|GetWorldLocation (Variables|Default|GetBreakLocation)) (+ (Transformation|GetWorldLocation (Variables|Default|GetBreakLocation)) (Math|Conversions|ToVector(Vector2D) (* (Variables|FellingResults|GetFallVector) (ScalableFloat|GetValueatLevel (Variables|FellingDataRows|GetFellingImpulseMagnitude))))) 0.0 "(R=1.000000,G=0.000000,B=1.000000,A=1.000000)" 999.0 8.0)))

(fn TrySpawnStump
  (if (Variables|FellingConfig|GetbDoesSpawnStump)
    (bind _returnvalue (Building|SpawnBuildingActor (Variables|FellingConfig|GetSpawnedStumpActorClass) (Variables|FellingResults|GetStartingWorldTransform))) ; Spawn stump
    (bind _output_get (Variables|FellingResults|SetSpawnedStumpObject _returnvalue))
    (CallFunction|ApplyChrome _output_get)
    (Default|CallOnSpawnedStumpObject _output_get)))

(fn ConstructionScript
  (Transformation|ResetRelativeTransform (Variables|Default|GetSpawnPreviewMesh)))
